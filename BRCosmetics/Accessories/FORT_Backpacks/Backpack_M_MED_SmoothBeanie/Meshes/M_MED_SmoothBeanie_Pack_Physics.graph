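# Physics asset for the SmoothBeanie backpack cosmetic, defining simulation bodies and joints for its dangling and moving parts.
# A box body on the pack connects through angularly free constraints to small base capsules, one per attachment.
# From these hang three three-segment dollar strands, a two-link chain ending in two box tags, a two-segment handle, a pocket and a torch.
# These dynamic joints use tight swing limits and near-zero twist, restricting their sway.
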PhysicsAsset /BRCosmetics/Accessories/FORT_Backpacks/Backpack_M_MED_SmoothBeanie/Meshes/M_MED_SmoothBeanie_Pack_Physics
bodies {
 base_chain_01 [{shapeType=Capsule, length=0.505, radius=0.505, shapeName=base_chain_01_capsule, rotation=<rot 0 -90 0>}]
 base_dollars_a [{shapeType=Capsule, length=0.505, radius=0.505, shapeName=base_dollars_a_capsule, rotation=<rot 0 -90 0>}]
 base_dollars_b [{shapeType=Capsule, length=0.505, radius=0.505, shapeName=base_dollars_b_capsule, rotation=<rot 0 -90 0>}]
 base_dollars_c [{shapeType=Capsule, length=0.505, radius=0.505, shapeName=base_dollars_c_capsule, rotation=<rot 0 -90 0>}]
 base_handle [{shapeType=Capsule, length=0.505, radius=0.505, shapeName=base_handle_capsule, rotation=<rot 0 -90 0>}]
 base_pocket [{shapeType=Capsule, length=0.505, radius=0.505, shapeName=base_pocket_capsule, rotation=<rot 0 -90 0>}]
 base_torch [{shapeType=Capsule, length=0.505, radius=0.505, shapeName=base_torch_capsule, rotation=<rot 0 -90 0>}]
 dyn_chain_01 [{shapeType=Capsule, length=0.5, radius=0.5, center=<1 0 0>, rotation=<rot 90 0 0>}]
 dyn_chain_02 [{shapeType=Capsule, length=0.5, radius=0.5, rotation=<rot 90 0 0>}]
 dyn_dollars_a_01 [{shapeType=Capsule, length=3, radius=0.5, center=<2 0 0>, rotation=<rot 90 0 0>}]
 dyn_dollars_a_02 [{shapeType=Capsule, length=3, radius=0.5, center=<2 0 0>, rotation=<rot 90 0 0>}]
 dyn_dollars_a_03 [{shapeType=Capsule, length=3, radius=0.5, center=<2 0 0>, rotation=<rot 90 0 0>}]
 dyn_dollars_b_01 [{shapeType=Capsule, length=3, radius=0.5, center=<2 0 0>, rotation=<rot 90 0 0>}]
 dyn_dollars_b_02 [{shapeType=Capsule, length=3, radius=0.5, center=<2 0 0>, rotation=<rot 90 0 0>}]
 dyn_dollars_b_03 [{shapeType=Capsule, length=3, radius=0.5, center=<2 0 0>, rotation=<rot 90 0 0>}]
 dyn_dollars_c_01 [{shapeType=Capsule, length=3, radius=0.5, center=<2 0 0>, rotation=<rot 90 0 0>}]
 dyn_dollars_c_02 [{shapeType=Capsule, length=3, radius=0.5, center=<2 0 0>, rotation=<rot 90 0 0>}]
 dyn_dollars_c_03 [{shapeType=Capsule, length=3, radius=0.5, center=<2 0 0>, rotation=<rot 90 0 0>}]
 dyn_handle_01 [{shapeType=Capsule, length=1, radius=1, center=<1 0 0>, rotation=<rot 90 0 0>}]
 dyn_handle_02 [{shapeType=Capsule, length=1, radius=1, center=<1 0 0>, rotation=<rot 90 0 0>}]
 dyn_pocket [{shapeType=Capsule, length=4, radius=2, shapeName=dyn_pocket_capsule, center=<2 0 0>, rotation=<rot 0 0 90>}]
 dyn_tag_a [{shapeType=Box, extentX=3, extentY=0.7, extentZ=5, shapeName=dyn_tag_a_box, center=<3.5 -0.0423 0.0118>, rotation=<rot 0 89.1299 -89.767>}]
 dyn_tag_b [{shapeType=Box, extentX=3, extentY=0.7, extentZ=5, shapeName=dyn_tag_b_box, center=<3.5 -0.1185 0.1159>, rotation=<rot 0 89.5715 -87.8891>}]
 dyn_torch [{shapeType=Capsule, length=4, radius=1.5, shapeName=dyn_torch_capsule, center=<2 0 0.0839>, rotation=<rot -90 0 0>}]
 pack [{
   shapeType=Box
   extentX=7.8775
   extentY=10.5692
   extentZ=37.3955
   shapeName=pack_box
   center=<-1.4015 -4.9586 -2.1318>
   rotation=<rot 0.0046 90.1074 -50.1828>
  }]
}
constraints [
 {bone1Name=dyn_chain_02, bone2Name=dyn_chain_01, swing1LimitDegrees=25, swing2LimitDegrees=25, twistLimitDegrees=0.15}
 {bone1Name=dyn_dollars_a_02, bone2Name=dyn_dollars_a_01, swing1LimitDegrees=12, swing2LimitDegrees=8, twistLimitDegrees=0}
 {bone1Name=dyn_dollars_a_03, bone2Name=dyn_dollars_a_02, swing1LimitDegrees=12, swing2LimitDegrees=8, twistLimitDegrees=0}
 {bone1Name=dyn_dollars_b_02, bone2Name=dyn_dollars_b_01, swing1LimitDegrees=12, swing2LimitDegrees=8, twistLimitDegrees=0}
 {bone1Name=dyn_dollars_b_03, bone2Name=dyn_dollars_b_02, swing1LimitDegrees=12, swing2LimitDegrees=8, twistLimitDegrees=0}
 {bone1Name=dyn_dollars_c_02, bone2Name=dyn_dollars_c_01, swing1LimitDegrees=8, swing2LimitDegrees=6, twistLimitDegrees=0}
 {bone1Name=dyn_dollars_c_03, bone2Name=dyn_dollars_c_02, swing1LimitDegrees=8, swing2LimitDegrees=6, twistLimitDegrees=0}
 {bone1Name=dyn_dollars_b_01, bone2Name=base_dollars_b, swing1LimitDegrees=12, swing2LimitDegrees=8, twistLimitDegrees=0}
 {bone1Name=dyn_dollars_c_01, bone2Name=base_dollars_c, swing1LimitDegrees=8, swing2LimitDegrees=6, twistLimitDegrees=0}
 {bone1Name=dyn_dollars_a_01, bone2Name=base_dollars_a, swing1LimitDegrees=12, swing2LimitDegrees=8, twistLimitDegrees=0}
 {bone1Name=dyn_chain_01, bone2Name=base_chain_01, swing1LimitDegrees=15, swing2LimitDegrees=15, twistLimitDegrees=0.2}
 {bone1Name=dyn_pocket, bone2Name=base_pocket, swing1LimitDegrees=10, swing2LimitDegrees=1, twistLimitDegrees=0}
 {bone1Name=dyn_torch, bone2Name=base_torch, swing1LimitDegrees=10, swing2LimitDegrees=5, twistLimitDegrees=0.2}
 {bone1Name=dyn_handle_01, bone2Name=base_handle, swing1LimitDegrees=15, swing2LimitDegrees=2, twistLimitDegrees=0}
 {bone1Name=dyn_handle_02, bone2Name=dyn_handle_01, swing1LimitDegrees=15, swing2LimitDegrees=2, twistLimitDegrees=0}
 {bone1Name=dyn_tag_a, bone2Name=dyn_chain_02, swing1LimitDegrees=10, swing2LimitDegrees=10, twistLimitDegrees=1}
 {bone1Name=dyn_tag_b, bone2Name=dyn_chain_02, swing1LimitDegrees=10, swing2LimitDegrees=10, twistLimitDegrees=1}
 {
  bone1Name=base_chain_01
  bone2Name=pack
  swing1LimitDegrees=45
  swing1Motion=Free
  swing2LimitDegrees=45
  swing2Motion=Free
  twistLimitDegrees=45
  twistMotion=Free
 }
 {
  bone1Name=base_dollars_a
  bone2Name=pack
  swing1LimitDegrees=45
  swing1Motion=Free
  swing2LimitDegrees=45
  swing2Motion=Free
  twistLimitDegrees=45
  twistMotion=Free
 }
 {
  bone1Name=base_dollars_b
  bone2Name=pack
  swing1LimitDegrees=45
  swing1Motion=Free
  swing2LimitDegrees=45
  swing2Motion=Free
  twistLimitDegrees=45
  twistMotion=Free
 }
 {
  bone1Name=base_dollars_c
  bone2Name=pack
  swing1LimitDegrees=45
  swing1Motion=Free
  swing2LimitDegrees=45
  swing2Motion=Free
  twistLimitDegrees=45
  twistMotion=Free
 }
 {
  bone1Name=base_handle
  bone2Name=pack
  swing1LimitDegrees=45
  swing1Motion=Free
  swing2LimitDegrees=45
  swing2Motion=Free
  twistLimitDegrees=45
  twistMotion=Free
 }
 {
  bone1Name=base_torch
  bone2Name=pack
  swing1LimitDegrees=45
  swing1Motion=Free
  swing2LimitDegrees=45
  swing2Motion=Free
  twistLimitDegrees=45
  twistMotion=Free
 }
 {
  bone1Name=base_pocket
  bone2Name=pack
  swing1LimitDegrees=45
  swing1Motion=Free
  swing2LimitDegrees=45
  swing2Motion=Free
  twistLimitDegrees=45
  twistMotion=Free
 }
]
